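# Physics asset for the Order Guard back bling backpack mesh, providing collision and joints for its hanging chain and hook.
# Two capsules form the pack body, and small capsules cover each of four chain bones and the hook.
# Angular-limited constraints link them in series from body through chain base to hook, with one wider swing axis and tight twist, loosening slightly toward the hook end.
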PhysicsAsset /BRCosmetics/Accessories/FORT_Backpacks/Backpack_M_MED_Order_Guard/Meshes/M_MED_Order_Guard_Pack_Physics
bodies {
 body [
  {shapeType=Capsule, length=21.306, radius=9.2172, shapeName=body_capsule, center=<-2.2603 0 0.3109>, rotation=<rot 0 0 90>}
  {shapeType=Capsule, length=7.5, radius=1.5, shapeName=body_capsule, center=<10.8842 -0.0513 -1.0247>, rotation=<rot -0.0001 0 90>}
 ]
 dyn_chain_01 [{shapeType=Capsule, length=0.5, radius=1, shapeName=dyn_chain_01_capsule, rotation=<rot 90 0 0>}]
 dyn_chain_02 [{shapeType=Capsule, length=0.5, radius=1, shapeName=dyn_chain_02_capsule, rotation=<rot 90 0 0>}]
 dyn_chain_03 [{shapeType=Capsule, length=0.5, radius=1, shapeName=dyn_chain_03_capsule, rotation=<rot 90 0 0>}]
 dyn_chain_base [{shapeType=Capsule, length=0.5, radius=1, shapeName=dyn_chain_base_capsule, rotation=<rot 90 0 0>}]
 dyn_hook [{shapeType=Capsule, length=0.5, radius=1, shapeName=dyn_hook_capsule, center=<0.8443 0 0>, rotation=<rot 90 0 0>}]
}
constraints [
 {bone1Name=dyn_chain_base, bone2Name=body, swing1LimitDegrees=5, swing2LimitDegrees=25, twistLimitDegrees=5}
 {bone1Name=dyn_chain_01, bone2Name=dyn_chain_base, swing1LimitDegrees=5, swing2LimitDegrees=25, twistLimitDegrees=5}
 {bone1Name=dyn_chain_02, bone2Name=dyn_chain_01, swing1LimitDegrees=5, swing2LimitDegrees=25, twistLimitDegrees=5}
 {bone1Name=dyn_chain_03, bone2Name=dyn_chain_02, swing1LimitDegrees=10, swing2LimitDegrees=25, twistLimitDegrees=5}
 {bone1Name=dyn_hook, bone2Name=dyn_chain_03, swing1LimitDegrees=10, swing2LimitDegrees=25, twistLimitDegrees=5}
]
