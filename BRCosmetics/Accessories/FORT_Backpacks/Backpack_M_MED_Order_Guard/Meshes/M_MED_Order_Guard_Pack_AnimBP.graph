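# Animation Blueprint for the Order Guard backpack cosmetic, adding physics-driven secondary motion to the pack.
# A RigidBody node simulates from the reference pose in root-bone space, using the pack's physics asset and overridden gravity.
# Additive bone transforms then roll dyn_chain_01 and, in world space, dyn_hook.
# The update event sets dynAlpha and skydiveAlpha from front-end and skydive-floating flags, but the AnimGraph reads neither variable.
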AnimBlueprint /BRCosmetics/Accessories/FORT_Backpacks/Backpack_M_MED_Order_Guard/Meshes/M_MED_Order_Guard_Pack_AnimBP
target_skeleton=&/M_MED_Order_Guard_Pack_Skeleton
graphs [
 AnimGraph {nodes [
   n1 = RigidBody(ComponentPose=Local To Component(LocalPose=Local Space Ref Pose())) {
    bClampLinearTranslationLimitToRefPose=true
    bOverrideWorldGravity=true
    overridePhysicsAsset=&/M_MED_Order_Guard_Pack_Physics
    simulationSpace=BaseBoneSpace
    alphaBoolBlend {blendOption=Linear}
    baseBoneRef {boneName=root}
    componentLinearAccScale=<0.1 0.1 0.1>
    overrideWorldGravity {z=-980}
   }
   n2 = "Transform (Modify) Bone - Bone: dyn_chain_01"(ComponentPose=@n1, Alpha=1, Translation="0,0,0", Rotation="0,0,90", Scale="1,1,1") {rotationMode=BMM_Additive, alphaBoolBlend {blendOption=Linear}, boneToModify {boneName=dyn_chain_01}, rotation {roll=90}}
   n3 = "Transform (Modify) Bone - Bone: dyn_hook"(ComponentPose=@n2, Alpha=1, Translation="0,0,0", Rotation="0,0,40", Scale="1,1,1") {
    rotationMode=BMM_Additive
    rotationSpace=BCS_WorldSpace
    alphaBoolBlend {blendOption=Linear}
    boneToModify {boneName=dyn_hook}
    rotation {roll=40}
   }
   Output Pose(Result=Component To Local(ComponentPose=@n3)) {layerGroup=None}
  ]}
 EventGraph {nodes [
   Event BlueprintUpdateAnimation()
   > n1 = Branch(Condition=Get bIsInFrontEnd())
   > Set dynAlpha(dynAlpha:double=0.5) -> (Output_Get:double)
   Set dynAlpha(execute=@n1.else, dynAlpha=1)
   > n2 = Branch(Condition=Get bIsSkydiveFloating())
   > Set skydiveAlpha(skydiveAlpha:double=1) -> (Output_Get:double)
   Set skydiveAlpha(execute=@n2.else, skydiveAlpha=0)
   TryGetPawnOwner()
  ]}
]
variables [
 dynAlpha {type=number}
 skydiveAlpha {type=number}
]
